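# Physics asset for the COD Squad Plaid Male pickaxe cosmetic, setting up a jointed string piece on the two-handed axe.
# Box bodies cover the handle, a small string master bone and the string itself.
# The constraints form a chain: the master rotates freely on the handle, and the string has asymmetric swing and tighter twist limits relative to the master.
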PhysicsAsset /BRCosmetics/Weapons/FORT_Melee/Pickaxe_CODSquadPlaidMale/Meshes/COD_Squad_Plaid_Male_Axe_PhysicsAsset
bodies {
 Melee_TwoHanded_Handle [{shapeType=Box, extentX=4.8, extentY=7.6372, extentZ=22.0397, center=<1.1 -1.7 50>}]
 dyn_string [{shapeType=Box, extentX=4.6, extentY=3.2, extentZ=2.2, center=<4.353 -2.553 -0.2625>, rotation=<rot -10 -45 10.0004>}]
 dyn_string_mstr [{shapeType=Box, extentX=1.01, extentY=1.01, extentZ=1.01}]
}
constraints [
 {
  bone1Name=dyn_string_mstr
  bone2Name=Melee_TwoHanded_Handle
  swing1LimitDegrees=45
  swing1Motion=Free
  swing2LimitDegrees=45
  swing2Motion=Free
  twistLimitDegrees=45
  twistMotion=Free
 }
 {bone1Name=dyn_string, bone2Name=dyn_string_mstr, swing1LimitDegrees=60, swing2LimitDegrees=75, twistLimitDegrees=15}
]
